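# Tow hook weapon reticle widget that draws the attachment point indicator, derived from FortWeaponAttachmentPointWidget.
# SetPointOfAttachmentState plays the vehicle-targeted animation for a valid target, otherwise the ADS or no-target animation based on isADS, which PlayADSAnimation sets.
# Animations morph one image through material parameters, opacity and scale.
# StartLockedOnLoop and StopLockedOnLoop sequencer events control a looping rotation animation for the locked-on look.
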WidgetBlueprint /TowHookWeapon/UI/Reticle/WBP_Reticle_TowHook_AttachmentPoint
animations [
 Anim_NoTarget {
  binding_count=1
  length_seconds=0.3
  object=&self:Anim_NoTarget
  sequence {
   bindings [Image_ReticlePointOfAttachment_1 {
     id=#1
     possessed_class=Image
     tracks [
      {
       type=MovieSceneWidgetMaterialTrack
       sections [{
         type=MovieSceneParameterSection
         end_frame=6
         start_frame=0
         channels [
          MiddleCircleSize {type=MovieSceneScriptingFloatChannel, const [-0.07,1]}
          OffsetX {type=MovieSceneScriptingFloatChannel, const [0.168,1]}
          OffsetY {type=MovieSceneScriptingFloatChannel, const [-0.196,1]}
          OffsetY_Top {type=MovieSceneScriptingFloatChannel, const [0.38,1]}
          MorphValue {type=MovieSceneScriptingFloatChannel, const [0,1]}
          Point3_y {type=MovieSceneScriptingFloatChannel, const [-0.8,1]}
          POA_RevealSwitch {type=MovieSceneScriptingFloatChannel, const [0,1]}
          Polygon Opacity {type=MovieSceneScriptingFloatChannel, const [1,1]}
          POA Opacity {type=MovieSceneScriptingFloatChannel, const [0,1]}
         ]
        }]
      }
      {
       type=MovieSceneFloatTrack
       display_name=Render Opacity
       sections [{type=MovieSceneFloatSection, end_frame=0, start_frame=0, channels [None {type=MovieSceneScriptingFloatChannel, const [0.4,1]}]}]
      }
     ]
    }]
   tracks [{
     type=MovieSceneEventTrack
     sections [{
       type=MovieSceneEventTriggerSection
       channels [None {type=MovieSceneScriptingEventChannel, const ["<Struct 'MovieSceneEvent' (0x4697a7eae60) {}>", 1]}]
      }]
    }]
  }
 }
 Anim_POA_Intro {
  binding_count=1
  length_seconds=0.2
  object=&self:Anim_POA_Intro
  sequence {bindings [Image_ReticlePointOfAttachment_1 {
     id=#2
     possessed_class=Image
     tracks [{
       type=MovieSceneFloatTrack
       display_name=Render Opacity
       sections [{type=MovieSceneFloatSection, end_frame=4, start_frame=0, channels [None {type=MovieSceneScriptingFloatChannel, keys [[0,0],[4,0.4]]}]}]
      }]
    }]}
 }
 Anim_POA_VehicleTargeted {
  binding_count=1
  length_seconds=0.4967
  object=&self:Anim_POA_VehicleTargeted
  sequence {
   bindings [Image_ReticlePointOfAttachment_1 {
     id=#3
     possessed_class=Image
     tracks [
      {
       type=MovieSceneWidgetMaterialTrack
       sections [{
         type=MovieSceneParameterSection
         end_frame=9
         start_frame=0
         channels [
          OffsetY {type=MovieSceneScriptingFloatChannel, keys [[0,-0.196],[1.0833,-0.326],[6.927,-0.156]]}
          OffsetX {type=MovieSceneScriptingFloatChannel, keys [[0,0.168],[1.0833,0.244],[6.927,-0.046]]}
          OffsetY_Top {type=MovieSceneScriptingFloatChannel, keys [[0,0.38],[1.0833,0.38],[6.927,0.28]]}
          Point3_y {type=MovieSceneScriptingFloatChannel, keys [[0,-0.8],[1.0833,-0.8],[6.927,0.4555]]}
          POA_RevealSwitch {type=MovieSceneScriptingFloatChannel, keys [[0,0],[2.252,0],[7.5117,1]]}
          Polygon Opacity {type=MovieSceneScriptingFloatChannel, keys [[3.4207,1],[5.174,0]]}
          POA Opacity {type=MovieSceneScriptingFloatChannel, keys [[3.4207,0],[5.174,1],[9.9347,1]]}
          POA Scale {type=MovieSceneScriptingFloatChannel, keys [[0,0.6675],[3.4207,0.8],[8.096,0.6]]}
          IsRotationOn? {type=MovieSceneScriptingFloatChannel, const [0,2]}
          Polygone Mask Size {type=MovieSceneScriptingFloatChannel, keys [[0,0.548],[3.4207,0.4487]]}
         ]
        }]
      }
      {
       type=MovieSceneFloatTrack
       display_name=Render Opacity
       sections [{type=MovieSceneFloatSection, end_frame=1, start_frame=0, channels [None {type=MovieSceneScriptingFloatChannel, keys [[0,0.4],[1.0833,0.9]]}]}]
      }
      {
       type=MovieScene2DTransformTrack
       sections [{
         type=MovieScene2DTransformSection
         end_frame=5
         start_frame=0
         channels [
          Scale.X {type=MovieSceneScriptingFloatChannel, keys [[0,1.5],[2,2.2],[5.961,1.5]]}
          Scale.Y {type=MovieSceneScriptingFloatChannel, keys [[0,1.5],[2,2.2],[5.961,1.5]]}
         ]
        }]
      }
     ]
    }]
   tracks [{
     type=MovieSceneEventTrack
     sections [{
       type=MovieSceneEventTriggerSection
       channels [None {type=MovieSceneScriptingEventChannel, const ["<Struct 'MovieSceneEvent' (0x4697a7ec940) {}>", 1]}]
      }]
    }]
  }
 }
 Anim_POA_ADS {
  binding_count=1
  length_seconds=0.25
  object=&self:Anim_POA_ADS
  sequence {
   bindings [Image_ReticlePointOfAttachment_1 {
     id=#4
     possessed_class=Image
     tracks [
      {
       type=MovieSceneFloatTrack
       display_name=Render Opacity
       sections [{type=MovieSceneFloatSection, end_frame=5, start_frame=0, channels [None {type=MovieSceneScriptingFloatChannel, keys [[0,0.4],[5,0.9]]}]}]
      }
      {
       type=MovieSceneWidgetMaterialTrack
       sections [{
         type=MovieSceneParameterSection
         end_frame=5
         start_frame=0
         channels [
          OffsetY {type=MovieSceneScriptingFloatChannel, keys [[0,-0.196],[5,-0.186]]}
          OffsetX {type=MovieSceneScriptingFloatChannel, keys [[0,0.168],[5,0.148]]}
          OffsetY_Top {type=MovieSceneScriptingFloatChannel, keys [[0,0.38],[5,0.37]]}
          Point3_y {type=MovieSceneScriptingFloatChannel, keys [[0,-0.8],[5,-0.4]]}
          Polygone Mask Size {type=MovieSceneScriptingFloatChannel, const [0.488,1]}
          POA_RevealSwitch {type=MovieSceneScriptingFloatChannel, const [0,1]}
          POA Opacity {type=MovieSceneScriptingFloatChannel, const [0,1]}
          POA Scale {type=MovieSceneScriptingFloatChannel, const [0.6,1]}
          MiddleCircleSize {type=MovieSceneScriptingFloatChannel, const [-0.07,1]}
          Polygon Opacity {type=MovieSceneScriptingFloatChannel, const [1,1]}
         ]
        }]
      }
      {
       type=MovieScene2DTransformTrack
       sections [{
         type=MovieScene2DTransformSection
         end_frame=0
         start_frame=0
         channels [
          Scale.X {type=MovieSceneScriptingFloatChannel, const [1.5,1]}
          Scale.Y {type=MovieSceneScriptingFloatChannel, const [1.5,1]}
         ]
        }]
      }
     ]
    }]
   tracks [{
     type=MovieSceneEventTrack
     sections [{
       type=MovieSceneEventTriggerSection
       channels [None {type=MovieSceneScriptingEventChannel, const ["<Struct 'MovieSceneEvent' (0x4697a7ebb80) {}>", 1]}]
      }]
    }]
  }
 }
 Anim_POA_VehicleTargeted_Rotation {
  binding_count=1
  length_seconds=2.4938
  object=&self:Anim_POA_VehicleTargeted_Rotation
  sequence {bindings [Image_ReticlePointOfAttachment_1 {
     id=#3
     possessed_class=Image
     tracks [
      {
       type=MovieSceneWidgetMaterialTrack
       sections [{
         type=MovieSceneParameterSection
         end_frame=49
         start_frame=0
         channels [
          OffsetY {type=MovieSceneScriptingFloatChannel, const [-0.156,1]}
          OffsetX {type=MovieSceneScriptingFloatChannel, const [-0.046,1]}
          OffsetY_Top {type=MovieSceneScriptingFloatChannel, const [0.28,1]}
          Point3_y {type=MovieSceneScriptingFloatChannel, const [0.4555,1]}
          POA_RevealSwitch {type=MovieSceneScriptingFloatChannel, const [1,1]}
          Polygon Opacity {type=MovieSceneScriptingFloatChannel, const [0,1]}
          POA Opacity {type=MovieSceneScriptingFloatChannel, const [1,1]}
          POA Scale {type=MovieSceneScriptingFloatChannel, const [0.6,1]}
          IsRotationOn? {type=MovieSceneScriptingFloatChannel, const [0,1]}
          Polygone Mask Size {type=MovieSceneScriptingFloatChannel, const [0.4487,1]}
          RotationAngle {type=MovieSceneScriptingFloatChannel, keys [[0,0],[49.8753,1]]}
         ]
        }]
      }
      {
       type=MovieSceneFloatTrack
       display_name=Render Opacity
       sections [{type=MovieSceneFloatSection, end_frame=5, start_frame=0, channels [None {type=MovieSceneScriptingFloatChannel, const [0.9,1]}]}]
      }
      {
       type=MovieScene2DTransformTrack
       sections [{
         type=MovieScene2DTransformSection
         end_frame=29
         start_frame=0
         channels [
          Translation.X {type=MovieSceneScriptingFloatChannel}
          Translation.Y {type=MovieSceneScriptingFloatChannel}
          Angle {type=MovieSceneScriptingFloatChannel}
          Scale.X {type=MovieSceneScriptingFloatChannel, const [1.5,1]}
          Scale.Y {type=MovieSceneScriptingFloatChannel, const [1.5,1]}
          Shear.X {type=MovieSceneScriptingFloatChannel}
          Shear.Y {type=MovieSceneScriptingFloatChannel}
         ]
        }]
      }
     ]
    }]}
 }
]
graphs [
 EventGraph {nodes [
   Event Destruct()
   > Parent: Destruct()
   > StopAllAnimations(self:User Widget Object Reference)
  ]}
 PlayADSAnimation {nodes [
   n1 = PlayADSAnimation!() -> (then, isTargetting:bool=~) {is=title K2Node_FunctionEntry}
   > n2 = Branch(Condition=@n1)
   n3 = Get Anim_POA_ADS() -> (Anim_POA_ADS:Widget Animation Object Reference)
   > PlayAnimationForward(InAnimation=@n3)
   > Set isADS(isADS:bool=true) -> (Output_Get:bool)
   PlayAnimationReverse(execute=@n2.else, InAnimation=@n3)
   > Set isADS(isADS=false)
  ]}
 Sequencer Events {nodes [
   StartLockedOnLoop() {is=title K2Node_CustomEvent}
   n1 = Get Anim_POA_VehicleTargeted_Rotation() -> (Anim_POA_VehicleTargeted_Rotation:Widget Animation Object Reference)
   > PlayAnimation(InAnimation=@n1, NumLoopsToPlay=0, bRestoreState=true)
   StopLockedOnLoop() {is=title K2Node_CustomEvent}
   > StopAnimation(InAnimation=@n1)
  ]}
 SetPointOfAttachmentState {nodes [
   n1 = SetPointOfAttachmentState!() -> (then, HasValidTarget:bool=~) {is=title K2Node_FunctionEntry}
   > StopAllAnimations()
   n2 = Select(
    Option 0:Widget Animation Object Reference=Select(
     Option 0=Get Anim_NoTarget() -> (Anim_NoTarget:Widget Animation Object Reference)
     Option 1=Get Anim_POA_ADS()
     Index=Get isADS() -> (isADS:bool)
    )
    Option 1:Widget Animation Object Reference=Get Anim_POA_VehicleTargeted() -> (Anim_POA_VehicleTargeted:Widget Animation Object Reference)
    Index=@n1
   ) -> (ReturnValue:Widget Animation Object Reference)
   > Is Valid(InputObject=@n2)
   > PlayAnimation(InAnimation=@n2)
  ]}
]
variables [isADS {type=boolean}]
widgets {
 info {inheritedWidgetCount=1, parentClass=/Script/FortniteUI.FortWeaponAttachmentPointWidget, rootWidgetClass=&Overlay, widgetCount=3}
 widgets [
  Overlay_LockedOn {widgetClassPath=&Overlay, bIsVariable=true, widget=Overlay_LockedOn}
  Image_ReticlePointOfAttachment_1 {
   widgetClassPath=&Image
   bIsVariable=true
   parent=Overlay_LockedOn
   widget=Image_ReticlePointOfAttachment_1
   properties {
    renderOpacity=0.4
    brush {
     bIsDynamicallyLoaded=false
     drawAs=Image
     imageType=FullColor
     mirroring=NoMirror
     resourceName=None
     resourceObject=/TowHookWeapon/UI/Materials/MI_UI_TowHookAttachment
     tiling=NoTile
     imageSize=<45 45>
     margin=<ltrb 0 0 0 0>
     outlineSettings {
      bUseBrushTransparency=false
      roundingType=HalfHeightRadius
      width=0
      color {colorUseRule=UseColor_Specified, specifiedColor=<rgba 0 0 0 0>}
      cornerRadii=<0 0 0 0>
     }
     tintColor {colorUseRule=UseColor_Specified, specifiedColor=<rgba 1 1 1 1>}
     uVRegion {bIsValid=false, max=<0 0>, min=<0 0>}
    }
    renderTransform {angle=0, scale=<1.5 1.5>, shear=<0 0>, translation=<0 0>}
   }
   slot {class=OverlaySlot, properties {horizontalAlignment=HAlign_Fill, verticalAlignment=VAlign_Fill}}
  }
  TutorialHighlight {widgetClassPath=&TutorialHighlight}
 ]
}
