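# Tow hook weapon reticle widget that layers a material-driven image and an attachment-point sub-widget, showing target validity and ADS state.
# On construct it plays an intro and binds to HUD state, plus the weapon's ammo, targeting and target-state delegates, unbinding on destruct.
# Target changes cross-fade the image and update the attachment point, with debug prints left in; empty ammo shows no target.
# HUD state hides elements during skydiving or DBNO.
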WidgetBlueprint /TowHookWeapon/UI/Reticle/WBP_Reticle_TowHookWeapon
animations [
 Anim_NoTarget {
  binding_count=1
  length_seconds=0.05
  object=&self:Anim_NoTarget
  sequence {bindings [Image_ReticleVisual {
     id=#1
     possessed_class=Image
     tracks [{
       type=MovieSceneColorTrack
       display_name=Color and Opacity
       sections [{
         type=MovieSceneColorSection
         channels [
          Color.R {type=MovieSceneScriptingFloatChannel, const [1,1]}
          Color.G {type=MovieSceneScriptingFloatChannel, const [1,1]}
          Color.B {type=MovieSceneScriptingFloatChannel, const [1,1]}
          Color.A {type=MovieSceneScriptingFloatChannel, keys [[0,0],[1,1]]}
         ]
        }]
      }]
    }]}
 }
 Anim_Intro {
  binding_count=2
  length_seconds=0.3
  object=&self:Anim_Intro
  sequence {bindings [
    Image_ReticleVisual {
     id=#2
     possessed_class=Image
     tracks [{
       type=MovieSceneFloatTrack
       display_name=Render Opacity
       sections [{type=MovieSceneFloatSection, end_frame=0, start_frame=0, channels [None {type=MovieSceneScriptingFloatChannel, const [0.8,1]}]}]
      }]
    }
    Reticle {
     id=#3
     possessed_class=Overlay
     tracks [
      {
       type=MovieSceneFloatTrack
       display_name=Render Opacity
       sections [{type=MovieSceneFloatSection, channels [None {type=MovieSceneScriptingFloatChannel, keys [[0,0],[6,1]]}]}]
      }
      {
       type=MovieScene2DTransformTrack
       sections [{
         type=MovieScene2DTransformSection
         channels [
          Translation.X {type=MovieSceneScriptingFloatChannel}
          Translation.Y {type=MovieSceneScriptingFloatChannel}
          Angle {type=MovieSceneScriptingFloatChannel}
          Scale.X {type=MovieSceneScriptingFloatChannel, keys [[0,1.25],[6,1]]}
          Scale.Y {type=MovieSceneScriptingFloatChannel, keys [[0,1.25],[6,1]]}
          Shear.X {type=MovieSceneScriptingFloatChannel}
          Shear.Y {type=MovieSceneScriptingFloatChannel}
         ]
        }]
      }
     ]
    }
   ]}
 }
 Anim_VehicleTargeted {
  binding_count=1
  length_seconds=0.1464
  object=&self:Anim_VehicleTargeted
  sequence {bindings [Image_ReticleVisual {
     id=#4
     possessed_class=Image
     tracks [{
       type=MovieSceneColorTrack
       display_name=Color and Opacity
       sections [{
         type=MovieSceneColorSection
         channels [
          Color.R {type=MovieSceneScriptingFloatChannel, const [1,1]}
          Color.G {type=MovieSceneScriptingFloatChannel, const [1,1]}
          Color.B {type=MovieSceneScriptingFloatChannel, const [1,1]}
          Color.A {type=MovieSceneScriptingFloatChannel, keys [[0,1],[2.9273,0]]}
         ]
        }]
      }]
    }]}
 }
 Anim_ADS {
  binding_count=1
  length_seconds=0.25
  object=&self:Anim_ADS
  sequence {bindings [Image_ReticleVisual {
     id=#5
     possessed_class=Image
     tracks [
      {
       type=MovieScene2DTransformTrack
       sections [{
         type=MovieScene2DTransformSection
         end_frame=5
         start_frame=0
         channels [
          Scale.X {type=MovieSceneScriptingFloatChannel, keys [[0,0.85],[5,1.15]]}
          Scale.Y {type=MovieSceneScriptingFloatChannel, keys [[0,0.85],[5,1.15]]}
         ]
        }]
      }
      {
       type=MovieSceneFloatTrack
       display_name=Render Opacity
       sections [{type=MovieSceneFloatSection, end_frame=5, start_frame=0, channels [None {type=MovieSceneScriptingFloatChannel, keys [[0,0.8],[5,0.4]]}]}]
      }
      {
       type=MovieSceneWidgetMaterialTrack
       sections [{
         type=MovieSceneParameterSection
         end_frame=5
         start_frame=0
         channels [ADS_ThicknessSwitch {type=MovieSceneScriptingFloatChannel, keys [[0,0],[5,1]]}]
        }]
      }
     ]
    }]}
 }
]
graphs [
 EventGraph {nodes [
   n1 = OnIsValidTargetChanged!() -> (OutputDelegate, then, bIsValidTarget:bool=~) {is=title K2Node_CustomEvent}
   > UpdateGrappleWeaponReticleAnimation!(execute, self, HasValidTarget:bool=@n1.bIsValidTarget) -> (then)
   > SetPointOfAttachmentState!(
    execute
    self:WBP Reticle Tow Hook Attachment Point Object Reference=Get WBP_Reticle_TowHook_AttachmentPoint() -> (WBP_Reticle_TowHook_AttachmentPoint:WBP Reticle Tow Hook Attachment Point Object Reference)
    HasValidTarget:bool=@n1.bIsValidTarget
   ) -> (then)
   Event Construct()
   > Parent: Construct()
   > PlayAnimation(InAnimation=Get Anim_Intro() -> (Anim_Intro:Widget Animation Object Reference))
   > Bind Event to On HUDState Refreshed(
    self:Fort HUDContext Object Reference=Get FortHUDContext!() -> (ReturnValue:Fort HUDContext Object Reference) {is=title K2Node_GetBlueprintContextEx}
    Delegate=Create Event()
   ) {is=title K2Node_AddDelegate}
   > n2 = Cast To B_TowHookWeapon_Athena!(
    execute
    Object=GetEquippedWeapon!(self:Fort Weapon Reticle Extension Widget Base Object Reference) -> (ReturnValue:Fort Weapon Object Reference)
   ) -> (then, CastFailed, AsB Tow Hook Weapon Athena:B Tow Hook Weapon Athena Object Reference) {is=title K2Node_DynamicCast}
   > n3 = Set Grapple Weapon Ref(Grapple Weapon Ref:B Tow Hook Weapon Athena Object Reference=@n2) -> (Output_Get:B Tow Hook Weapon Athena Object Reference)
   > Bind Event to On Local Ammo Changed(self:Fort Weapon Object Reference=@n3, Delegate=Create Event()) {is=title K2Node_AddDelegate}
   > Bind Event to On Targeting Changed(
    self:Fort Weapon Object Reference=Get Grapple Weapon Ref() -> (Grapple Weapon Ref:B Tow Hook Weapon Athena Object Reference)
    Delegate=Create Event()
   ) {is=title K2Node_AddDelegate}
   > n4 = Set TargetStateComponent(TargetStateComponent:Fort Weapon Component Target State Object Reference=GetComponentByClass!(
     self:Actor Object Reference=Get Grapple Weapon Ref()
     ComponentClass:Actor Component Class Reference=/Script/FortniteGame.FortWeaponComponent_TargetState
    ) -> (ReturnValue:Fort Weapon Component Target State Object Reference)) -> (Output_Get:Fort Weapon Component Target State Object Reference)
   > Is Valid(InputObject=@n4)
   > Bind Event to On Is Valid Target Changed(self:Fort Weapon Component Target State Object Reference=@n4, Delegate=Create Event()) {is=title K2Node_AddDelegate}
   K2Node_Event_2 = OnLocalAmmoChanged!() -> (OutputDelegate, then, LocalCount:int, LocalRemaining:int) {is=title K2Node_CustomEvent}
   > Branch(Condition="Equal (Integer)"(A=@K2Node_Event_2.LocalCount))
   > PlayAnimation(InAnimation=Get Anim_NoTarget() -> (Anim_NoTarget:Widget Animation Object Reference))
   > SetPointOfAttachmentState(self=Get WBP_Reticle_TowHook_AttachmentPoint(), HasValidTarget=false)
   n5 = Event PreConstruct()
   > Parent: PreConstruct(IsDesignTime=@n5.IsDesignTime)
   > Branch(Condition=@n5.IsDesignTime)
   n6 = Get Image_ReticleVisual() -> (Image_ReticleVisual:Image Object Reference)
   n7 = Get!(self:Image Object Reference=@n6) -> (
    Brush_TintColor:Slate Color Structure
    Brush_DrawAs:ESlateBrushDrawType Enum
    Brush_Tiling:ESlateBrushTileType Enum
    Brush_Mirroring:ESlateBrushMirrorType Enum
    Brush_ImageSize:Vector 2D Structure
    Brush_Margin:Margin Structure
    Brush_ResourceObject:Object Reference
    Brush_OutlineSettings:Slate Brush Outline Settings Structure
   )
   > SetBrush(
    self=@n6
    InBrush=Make Slate Brush!(
     TintColor:Slate Color Structure=@n7
     DrawAs:ESlateBrushDrawType Enum=@n7.Brush_DrawAs
     Tiling:ESlateBrushTileType Enum=@n7.Brush_Tiling
     Mirroring:ESlateBrushMirrorType Enum=@n7.Brush_Mirroring
     ImageSize:Vector 2D Structure=Get ReticleSize() -> (ReticleSize:Vector 2D Structure)
     Margin:Margin Structure=@n7.Brush_Margin
     ResourceObject:Object Reference=Get ReticleMaterial() -> (ReticleMaterial:Material Instance Object Reference)
     OutlineSettings:Slate Brush Outline Settings Structure=@n7.Brush_OutlineSettings
    ) -> (SlateBrush:Slate Brush Structure) {is=title K2Node_MakeStruct}
   )
   Event Destruct()
   > Parent: Destruct()
   > n8 = Sequence()
   > Unbind Event from On HUDState Refreshed(self:Fort HUDContext Object Reference=Get FortHUDContext(), Delegate=Create Event()) {is=title K2Node_RemoveDelegate}
   n9 = Get Grapple Weapon Ref()
   > Is Valid(InputObject=@n9)
   > Unbind Event from On Local Ammo Changed(self:Fort Weapon Object Reference=@n9, Delegate=Create Event()) {is=title K2Node_RemoveDelegate}
   > Unbind Event from On Targeting Changed(self:Fort Weapon Object Reference=@n9, Delegate=Create Event()) {is=title K2Node_RemoveDelegate}
   n10 = Get!(execute:Exec=@n8.then_1) -> (then:Exec, else:Exec, TargetStateComponent:Fort Weapon Component Target State Object Reference)
   > Unbind Event from On Is Valid Target Changed(self:Fort Weapon Component Target State Object Reference=@n10, Delegate=Create Event()) {is=title K2Node_RemoveDelegate}
  ]}
 OnHUDStateRefreshed {nodes [
   n1 = OnHUDStateRefreshed!() -> (then, NewHUDState:"Fort HUD State Structure (by ref)") {is=title K2Node_FunctionEntry}
   > n2 = Sequence()
   n3 = Break Fort HUD State!(FortHUDState:"Fort HUD State Structure (by ref)"=@n1) -> (
    bInBuildMode:bool=~
    bInCombatMode:bool=~
    bInEditMode:bool=~
    bInCreativeMode:bool=~
    bIsParachuteOpen:bool=~
    bIsFreeFalling:bool=~
    bInLockedBus:bool=~
    bInUnlockedBus:bool=~
    bOnTargeting:bool=~
    bOnUsingScopeTargeting:bool=~
    bOnCanTarget:bool=~
    bOnCanUseScopeTargeting:bool=~
    bOnCanUseSecondaryAbility:bool=~
    bCanOpenChute:bool=~
    bCrouching:bool=~
    bDBNO:bool=~
    bIsControllingRCPawn:bool=~
    bIsInVehicle:bool=~
    bIsDrivingVehicle:bool=~
    bCanSwapSeats:bool=~
    bIsCarryingDBNO:bool=~
    bIsCarryingHeldObject:bool=~
    bIsZiplining:bool=~
    bIsCreativeFlying:bool=~
    bImprovedDBNO:bool=~
   ) {is=title K2Node_BreakStruct}
   > SetVisibility(
    self=Get WBP_Reticle_TowHook_AttachmentPoint()
    InVisibility=Select(Option 0=HitTestInvisible, Option 1=Hidden, Index=OR Boolean(A=@n3.bIsParachuteOpen, B=@n3.bIsFreeFalling))
   )
   SetVisibility(execute=@n2.then_1, self=Get Image_ReticleVisual(), InVisibility=Select(Option 0=HitTestInvisible, Option 1=Hidden, Index=@n3.bDBNO))
  ]}
 SetTargetingState {nodes [
   n1 = SetTargetingState!() -> (then, bIsTargeting:bool=~) {is=title K2Node_FunctionEntry}
   > PlayADSAnimation!(execute, self:WBP Reticle Tow Hook Attachment Point Object Reference=Get WBP_Reticle_TowHook_AttachmentPoint(), isTargetting:bool=@n1) -> (then)
   > n2 = Branch(Condition=@n1)
   n3 = Get Anim_ADS() -> (Anim_ADS:Widget Animation Object Reference)
   > PlayAnimationForward(InAnimation=@n3)
   PlayAnimationReverse(execute=@n2.else, InAnimation=@n3)
  ]}
 UpdateGrappleWeaponReticleAnimation {nodes [
   n1 = UpdateGrappleWeaponReticleAnimation!() -> (then, HasValidTarget:bool=~) {is=title K2Node_FunctionEntry}
   > n2 = Branch(Condition=@n1)
   > StopAnimation(InAnimation=Get Anim_NoTarget())
   > PlayAnimationForward(InAnimation=Get Anim_VehicleTargeted() -> (Anim_VehicleTargeted:Widget Animation Object Reference))
   > PrintString!(execute, InString="Has Vehicle Target!!!!!!!", bPrintToScreen, bPrintToLog, TextColor, Duration, Key) -> (then)
   StopAnimation(execute=@n2.else, InAnimation=Get Anim_VehicleTargeted())
   > PlayAnimationForward(InAnimation=Get Anim_NoTarget())
   > PrintString(InString="No Target!!!!!")
  ]}
]
variables [
 bLatestTargetingLedge {type=boolean}
 bUpdateReticleColor {type=boolean}
 bUpdateReticleLocation {type=boolean, default=true}
 ChargeRotationParam {type=string, default=Charge Rotation}
 ChargesAvailableParam {type=string, default=Charges Available}
 FillProgressParam {type=string, default=Fill Progress}
 FinalChargeHighlightAnimParam {type=string, default=Final Charge Highlight Anim}
 FinalChargeOpacityParam {type=string, default=Final Charge Opacity}
 FirstChargeHighlightAnimParam {type=string, default=First Charge Highlight Anim}
 Grapple Weapon Ref {type=B_TowHookWeapon_Athena_C, default=null}
 LatestWorldLocation {type=Vector}
 ReticleMaterial {type=MaterialInstance, default=/TowHookWeapon/UI/Materials/MI_UI_TowHookReticle}
 ReticleSize {type=Vector2D, default=<150 150>}
 TargetingLedge {type=boolean}
 TargetStateComponent {type=FortWeaponComponent_TargetState, default=null}
]
widgets {
 info {
  inheritedWidgetCount=1
  parentClass=/Script/FortniteUI.FortWeaponReticleExtensionWidgetBase
  rootWidgetClass=&Overlay
  widgetCount=4
 }
 widgets [
  Reticle {widgetClassPath=&Overlay, widget=Reticle}
  Image_ReticleVisual {
   widgetClassPath=&Image
   bIsVariable=true
   parent=Reticle
   widget=Image_ReticleVisual
   properties {
    renderOpacity=0.8
    brush {
     bIsDynamicallyLoaded=false
     drawAs=Image
     imageType=NoImage
     mirroring=NoMirror
     resourceName=None
     resourceObject=&MI_UI_TowHookReticle
     tiling=NoTile
     imageSize=<80 80>
     margin=<ltrb 0 0 0 0>
     outlineSettings {
      bUseBrushTransparency=false
      roundingType=HalfHeightRadius
      width=0
      color {colorUseRule=UseColor_Specified, specifiedColor=<rgba 0 0 0 0>}
      cornerRadii=<0 0 0 0>
     }
     tintColor {colorUseRule=UseColor_Specified, specifiedColor=<rgba 1 1 1 1>}
     uVRegion {bIsValid=false, max=<0 0>, min=<0 0>}
    }
    renderTransform {angle=0, scale=<0.85 0.85>, shear=<0 0>, translation=<0 0>}
   }
   slot {class=OverlaySlot, properties {horizontalAlignment=HAlign_Center, verticalAlignment=VAlign_Center}}
  }
  WBP_Reticle_TowHook_AttachmentPoint {
   widgetClassPath=&/WBP_Reticle_TowHook_AttachmentPoint
   bIsVariable=true
   parent=Reticle
   widget=WBP_Reticle_TowHook_AttachmentPoint
   properties {renderTransform {angle=0, scale=<1.1 1.1>, shear=<0 0>, translation=<0 0>}}
   slot {class=OverlaySlot, properties {horizontalAlignment=HAlign_Center, verticalAlignment=VAlign_Center}}
  }
  TutorialHighlight {widgetClassPath=&TutorialHighlight}
 ]
}
